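# Sound Cue for distant robotic servo movement one-shot spot ambience at the Seven Workshop POI.
# A Random node picks among eight equally weighted stereo wave variations.
# A Modulator at the root varies pitch and volume on each play.
# The cue plays quietly in the ambient one-shots sound class and references a shared, more positional tile-ambient attenuation asset.
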SoundCue /Game/Sounds/Artemis/POI/Seven_Workshop/Spots/Spot_SevenWorkshop_Robotic_Servo_Movement_Distant_Cue
root=Modulator_0
nodes [
 {id=Modulator_0, children [Random_0], properties {pitchMin=0.8, volumeMax=0.9, volumeMin=0.7}}
 {
  id=Random_0
  children [WavePlayer_14, WavePlayer_15, WavePlayer_16, WavePlayer_17, WavePlayer_18, WavePlayer_19, WavePlayer_20, WavePlayer_21]
  properties {weights [1,1,1,1,1,1,1,1]}
 }
 {
  id=WavePlayer_14
  properties {soundWaveAssetPtr=&/Spot_SevenWorkshop_Robotic_Servo_Movement_Distant_08_Stereo}
 }
 {
  id=WavePlayer_15
  properties {soundWaveAssetPtr=&/Spot_SevenWorkshop_Robotic_Servo_Movement_Distant_01_Stereo}
 }
 {
  id=WavePlayer_16
  properties {soundWaveAssetPtr=&/Spot_SevenWorkshop_Robotic_Servo_Movement_Distant_02_Stereo}
 }
 {
  id=WavePlayer_17
  properties {soundWaveAssetPtr=&/Spot_SevenWorkshop_Robotic_Servo_Movement_Distant_03_Stereo}
 }
 {
  id=WavePlayer_18
  properties {soundWaveAssetPtr=&/Spot_SevenWorkshop_Robotic_Servo_Movement_Distant_04_Stereo}
 }
 {
  id=WavePlayer_19
  properties {soundWaveAssetPtr=&/Spot_SevenWorkshop_Robotic_Servo_Movement_Distant_05_Stereo}
 }
 {
  id=WavePlayer_20
  properties {soundWaveAssetPtr=&/Spot_SevenWorkshop_Robotic_Servo_Movement_Distant_06_Stereo}
 }
 {
  id=WavePlayer_21
  properties {soundWaveAssetPtr=&/Spot_SevenWorkshop_Robotic_Servo_Movement_Distant_07_Stereo}
 }
]
properties {
 attenuationSettings=/Game/Sounds/Attenuation/Fort_TileAmbient_OneShot_MorePositional_Att
 maxDistance=8000
 soundClassObject=&Fort_Ambient_OneShots
 volumeMultiplier=0.251
 attenuationOverrides {falloffDistance=6000, stereoSpread=10000, attenuationShapeExtents {x=2000}}
}
